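# Material function for the Prodigy Sage pickaxe that generates a rectangular mask from input texture coordinates and four edge positions.
# Each edge position is subtracted from the U or V coordinate, then passed through a very narrow SmoothStep for a near-hard threshold.
# Lerps invert the Right and Down thresholds and gate them by Left and Up, forming horizontal and vertical bands that multiply into the SquareMask output.
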MaterialFunction /BRCosmetics/Weapons/FORT_Melee/Pickaxe_ProdigySage/Materials/MaterialFuncions/MF_Pickaxe_ProdigySage_SquareMaskGenerator
graph {nodes [
  n1 = FunctionInput() {bUsePreviewValueAsDefault=true, inputName=X Pos Right, inputType=FunctionInput_Scalar, previewValue=<0.5 0 0 1>}
  n2 = FunctionInput() {inputName=Texture Coordinate, inputType=FunctionInput_Vector2}
  n3 = ComponentMask(None=Add(A=Multiply(A=@n1) {constB=-1}, B=@n2)) {r=true}
  n4 = LinearInterpolate(Alpha=Saturate(None=SmoothStep(Value=@n3) {constMax=0.001})) {constA=1, constB=0}
  n5 = FunctionInput() {bUsePreviewValueAsDefault=true, inputName=X Pos Left, inputType=FunctionInput_Scalar, sortPriority=2, previewValue=<0.2 0 0 1>}
  n6 = Add(A=@n2, B=Multiply(A=@n5) {constB=-1})
  n7 = Saturate(None=SmoothStep(Value=ComponentMask(None=@n6) {r=true}) {constMax=0.001})
  n8 = FunctionInput() {bUsePreviewValueAsDefault=true, inputName=Y Pos Down, inputType=FunctionInput_Scalar, sortPriority=4, previewValue=<0.9 0 0 1>}
  n9 = ComponentMask(None=Add(A=@n2, B=Multiply(A=@n8) {constB=-1})) {g=true}
  n10 = LinearInterpolate(Alpha=Saturate(None=SmoothStep(Value=@n9) {constMax=0.001})) {constA=1, constB=0}
  n11 = FunctionInput() {bUsePreviewValueAsDefault=true, inputName=Y Pos Up, inputType=FunctionInput_Scalar, sortPriority=3, previewValue=<0.2 0 0 1>}
  n12 = Add(A=@n2, B=Multiply(A=@n11) {constB=-1})
  n13 = Saturate(None=SmoothStep(Value=ComponentMask(None=@n12) {g=true}) {constMax=0.001})
  FunctionOutput(None=Multiply(A=LinearInterpolate(B=@n4, Alpha=@n7) {constB=0}, B=LinearInterpolate(B=@n10, Alpha=@n13) {constB=0})) {outputName=SquareMask}
 ]}
